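# Translucent, unlit, two-sided mesh-particle material for scrolling airflow streaks on a cylinder in the CyberScavenger glider FX.
# Two panning noise textures on optionally rotated UVs multiply together, with Dynamic Parameters driving tiling, speed and offset.
# That noise is masked by particle alpha, curve-atlas view-angle and V-axis trail fades, and a sine fade at both U edges.
# Particle color feeds a shared emissive-adjustment function.
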Material /Game/Gadgets/Assets/VinderTech_GliderChute/Glider_CyberScavenger/FX/M_CyberScavenger_CylinderAirFlow
graph {
 properties {
  bUsedWithMeshParticles=true
  bUsedWithNiagaraMeshParticles=true
  blendMode=BLEND_Translucent
  pixelDepthOffsetMode=PDOM_Legacy
  shadingModel=MSM_Unlit
  twoSided=true
  usedShadingModels=Unlit
 }
 nodes [
  n1 = ParticleColor()
  MaterialFunctionCall(Emissive=@n1.RGB) {materialFunction=&MF_FX_EmissiveAdjustments}
  MaterialFunctionCall!(UVs, Texture Size) {materialFunction=/Engine/Functions/Engine_MaterialFunctions01/Texturing/ComputeMipLevel}
  n2 = VectorParameter() {group=UV_tile, parameterName=UV_tile1, defaultValue=<rgba 1 0.3 0 1>}
  n3 = DynamicParameter() {paramNames [UV_tile,SpeedMult,UV_Offset,Param4]}
  n4 = TextureCoordinate()
  n5 = StaticSwitchParameter(
   True=MaterialFunctionCall(UVs=@n4, "Rotation Angle (0-1)"=ScalarParameter() {defaultValue=-0.1, parameterName=UV_rotate}) {materialFunction=&CustomRotator}.Rotated Values
   False=@n4
  ) {defaultValue=true, parameterName=UV_Rotate}
  n6 = Multiply(A=@n5, B=Multiply(A=ComponentMask(None=@n2.RGB) {g=true, r=true}, B=@n3.UV_tile))
  n7 = ComponentMask(None=VectorParameter() {group=UV_Speed, parameterName=Speed_UV1, defaultValue=<rgba 0 0.4 0 1>}.RGB) {g=true, r=true}
  n8 = TextureSampleParameter2D(UVs=Panner(Coordinate=Add(A=@n6, B=@n3.UV_Offset), Speed=Multiply(A=@n7, B=@n3.SpeedMult)) {speedY=1}) {
   group=Texture
   parameterName=Texture1
   samplerType=SAMPLERTYPE_LinearColor
   texture=/Game/Weapons/WeaponSkins/Textures/T_Swirly_Noise_highContrast1
  }
  n9 = VectorParameter() {group=UV_tile, parameterName=UV_tile2, sortPriority=1, defaultValue=<rgba 0.5 0.2 0 1>}
  n10 = Multiply(A=@n5, B=Multiply(A=@n3.UV_tile, B=ComponentMask(None=@n9.RGB) {g=true, r=true}))
  n11 = ComponentMask(None=VectorParameter() {group=UV_Speed, parameterName=Speed_UV2, sortPriority=1, defaultValue=<rgba 0 0.2 0 1>}.RGB) {g=true, r=true}
  n12 = TextureSampleParameter2D(UVs=Panner(Coordinate=Add(A=@n10, B=@n3.UV_Offset), Speed=Multiply(A=@n3.SpeedMult, B=@n11)) {speedY=1}) {group=Texture, parameterName=Texture2, sortPriority=1, texture=/Game/Weapons/FORT_Melee/Effects/LaserSwordFounders/T_LaserSwordTrailNoise}
  n13 = Multiply(A=@n8.R, B=@n12.R)
  n14 = VertexNormalWS()
  n15 = CameraVectorWS()
  n16 = CurveAtlasRowParameter(CurveTime=Abs(None=DotProduct(A=@n14, B=@n15))) {
   atlas=&AthenaCurveAtlas01
   curve=&/Curve_CyberScavenger_Glider
   defaultValue=89
   parameterName=FrenelCtrl
  }
  n17 = TextureCoordinate()
  n18 = CurveAtlasRowParameter(CurveTime=Power(Base=ComponentMask(None=@n17) {g=true}, Exp=ScalarParameter() {defaultValue=1, parameterName=Mask_Power})) {
   atlas=&AthenaCurveAtlas01
   curve=&Curve_CyberScavenger_Glider
   defaultValue=89
   parameterName=Trail_FadeCtrl
  }
  n19 = ComponentMask(None=@n17) {r=true}
  n20 = Sine(None=Multiply(A=@n19, B=OneMinus(None=@n19)))
  Multiply(A=Multiply(A=@n1.A, B=Power(Base=@n13)), B=Multiply(A=@n16.R, B=Multiply(A=@n18.A, B=@n20)))
 ]
}
